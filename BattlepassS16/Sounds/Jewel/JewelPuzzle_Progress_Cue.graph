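# Sound cue for progress feedback in the Season 16 battle pass jewel puzzle, built from reused CatchOllie quest flight sounds.
# A mixer layers two randomized, pitched-up branches: one of three flight-start waves plays immediately, and one of three in-air waves follows after a short fixed delay.
# It routes to the quests sound class, uses the jewel puzzle attenuation asset, and overrides concurrency to cap simultaneous instances.
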SoundCue /BattlepassS16/Sounds/Jewel/JewelPuzzle_Progress_Cue
root=Mixer_0
nodes [
 {id=Delay_0, children [Random_2], properties {delayMax=0.3, delayMin=0.3}}
 {id=Mixer_0, children [Modulator_1,Modulator_2], properties {inputVolume [1,1]}}
 {id=Modulator_1, children [Random_0], properties {pitchMax=1.6, pitchMin=1.4}}
 {id=Modulator_2, children [Delay_0], properties {pitchMax=1.6, pitchMin=1.4}}
 {id=Random_0, children [WavePlayer_0,WavePlayer_2,WavePlayer_6], properties {weights [1,1,1]}}
 {id=Random_2, children [WavePlayer_7,WavePlayer_8,WavePlayer_9], properties {weights [1,1,1]}}
 {id=WavePlayer_0, properties {soundWaveAssetPtr=/Game/Athena/Sounds/Quests/CatchOllie/CatchOllie_FlightStart_01}}
 {id=WavePlayer_2, properties {soundWaveAssetPtr=/Game/Athena/Sounds/Quests/CatchOllie/CatchOllie_FlightStart_02}}
 {id=WavePlayer_6, properties {soundWaveAssetPtr=/Game/Athena/Sounds/Quests/CatchOllie/CatchOllie_FlightStart_03}}
 {id=WavePlayer_7, properties {soundWaveAssetPtr=/Game/Athena/Sounds/Quests/CatchOllie/CatchOllie_FlightInAir_01}}
 {id=WavePlayer_8, properties {soundWaveAssetPtr=/Game/Athena/Sounds/Quests/CatchOllie/CatchOllie_FlightInAir_02}}
 {id=WavePlayer_9, properties {soundWaveAssetPtr=/Game/Athena/Sounds/Quests/CatchOllie/CatchOllie_FlightInAir_03}}
]
properties {
 attenuationSettings=&/JewelPuzzle_Att
 bOverrideConcurrency=true
 maxDistance=800
 soundClassObject=&Fort_Quests
 volumeMultiplier=1.68
 attenuationOverrides {distanceAlgorithm=Logarithmic, falloffDistance=2700, attenuationShapeExtents {x=300}}
 concurrencyOverrides {maxCount=3}
}
