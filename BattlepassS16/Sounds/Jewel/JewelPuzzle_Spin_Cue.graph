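# Sound cue for the spin sound of the Season 16 battle pass jewel puzzle, routed to the Fort_Quests sound class.
# A random node picks evenly between two spin wave variants and feeds a modulator that applies slight pitch variation.
# Playback uses reduced volume, the JewelPuzzle_Att attenuation asset and a concurrency override that caps simultaneous instances at two.
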SoundCue /BattlepassS16/Sounds/Jewel/JewelPuzzle_Spin_Cue
root=Modulator_1
nodes [
 {id=Modulator_1, children [Random_0], properties {pitchMax=1.02, pitchMin=0.98}}
 {id=Random_0, children [WavePlayer_0,WavePlayer_2], properties {weights [1,1]}}
 {id=WavePlayer_0, properties {soundWaveAssetPtr=&/Jewel_Puzzle_Spin_01_v1}}
 {id=WavePlayer_2, properties {soundWaveAssetPtr=&/Jewel_Puzzle_Spin_01_v2}}
]
properties {
 attenuationSettings=&/JewelPuzzle_Att
 bOverrideConcurrency=true
 maxDistance=800
 soundClassObject=&Fort_Quests
 volumeMultiplier=0.224
 attenuationOverrides {distanceAlgorithm=Logarithmic, falloffDistance=3000, attenuationShapeExtents {x=300}}
 concurrencyOverrides {maxCount=2}
}
